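; Water-walking detection gameplay ability for MotoWrestle that extends its parent so water stays solid when the player ice slides.
; On a falling movement change, it waits for the IceSliding tag, then clears the falling timer and restores capsule blocking against water unless the ocean is ignored.
; Fall-damage prevention takes the same blocking path when already sliding, otherwise deferring to the parent.
; Ending the ability ends the tag-wait task.
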(event EventOnFallingMovementChange
  ;[0.0137 0.1361 0.15 0.5] Check for ice slide when falling through air
  (|CallParentFunction)
  (CallFunction|CheckForIceSliding))

(event EventPreventWaterFallDamage
  ;[0.04286 0.15 0.1073 0.5] Set water back to blocking if performing ice slide. Otherwise allow parent to set water to overlap
  (bind _avatarplayerpawn (Variables|Transient|GetAvatarPlayerPawn)
    (:then
      (if (not (GameplayTags|HasMatchingGameplayTag _avatarplayerpawn (Variables|Default|GetTag_IceSliding)))
        (|CallParentFunction)
        (else
          (CallFunction|BlockWaterforSlide))))
    (:else)))

(event Custom|Check%20for%20Ice%20Sliding
  ;[0.06217 0.02382 0.15 0.5] Starts a Wait to see if the player has started Ice Sliding
  (bind _avatarplayerpawn (Variables|Transient|GetAvatarPlayerPawn)
    (:then
      (bind _asynctaskproxy (Ability|Tasks|WaitGameplayTagAdd (Variables|Default|GetTag_IceSliding) _avatarplayerpawn)
        (:then
          (Variables|Default|SetAsyncTask_BlockWaterForSlide _asynctaskproxy))
        (:Added
          (CallFunction|BlockWaterforSlide))))
    (:else)))

(event Custom|Block%20Water%20For%20Slide
  ;[0.1226 0 0.15 0.5] Sets the water collision back to blocking and invalidates the fall timer if the player begins sliding
  (Utilities|Time|ClearandInvalidateTimerbyHandle (Variables|Transient|GetFallingTimer))
  (bind _avatarplayerpawn (Variables|Transient|GetAvatarPlayerPawn)
    (:then
      (if (and (Utilities|Enum|NotEqual(Enum) (Collision|GetCollisionResponsetoChannel (Variables|/Script/Engine.Character|GetCapsuleComponent _avatarplayerpawn) "ECC_GameTraceChannel18") "ECR_Block") (not (Variables|Transient|GetbIgnoringOcean)))
        (CallFunction|BlockWater _avatarplayerpawn)))
    (:else)))

(event Ability|EventOnEndAbility
  (params (bWasCancelled :type Boolean))
  ;[0.15 0.0724 0.008451 0.5] On End Ability
  (|CallParentFunction bWasCancelled)
  (bind _asynctask_blockwaterforslide (Variables|Default|GetAsyncTask_BlockWaterForSlide)
    (:then
      (GameplayTasks|EndTask _asynctask_blockwaterforslide))
    (:else)))
